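# Activity browser modal, built on FortActivityBrowserAttributionsModal, that shows an island's attributions or credits in a UIKit dialog.
# Island name, creator, code and Back sit beside a scrolling rich-text panel, with a separate, initially collapsed link button.
# Blueprint events play a whoosh on activation, set the title per mode, and zero the backplate material's R/G corner radii.
# Override functions supply localized formats for attribution, creator and link code text.
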WidgetBlueprint /Game/UI/DiscoveryShared/UEFNAttributions/ActivityBrowserAttributionsModal
graphs [
 EventGraph {nodes [
   Event On Activated()
   > SpawnSound2D!(
    execute
    Sound:Sound Base Object Reference=Get IntroWhooshSound() -> (IntroWhooshSound:Sound Base Object Reference)
    VolumeMultiplier:float=1
    PitchMultiplier:float=1
    StartTime:float=0
    ConcurrencySettings:Sound Concurrency Object Reference
    bPersistAcrossLevelTransition:bool=false
    bAutoDestroy:bool=true
   ) -> (then, ReturnValue:Audio Component Object Reference)
   Event PreConstruct()
   > n1 = GetDynamicMaterial(self=Get Backplate!(self:WBP UIKit Backplate Base Object Reference=Get Backplate_Attributions() -> (Backplate_Attributions:WBP UIKit Backplate Object Reference)) -> (Backplate:Image Object Reference))
   > n2 = Get Vector Parameter Value!(execute, self:Material Instance Dynamic Object Reference=@n1, ParameterName:Name=Corner Radius) -> (then, ReturnValue_R:float=~, ReturnValue_G:float=~, ReturnValue_B:float=~, ReturnValue_A:float=~) {is=title K2Node_CallFunction}
   > SetVectorParameterValue!(
    execute
    self=@n1
    ParameterName=Corner Radius
    Value_R:float=0
    Value_G:float=0
    Value_B:float=@n2.ReturnValue_B
    Value_A:float=@n2.ReturnValue_A
   ) -> (then)
   Event OnAttributionsModalInitialized() {is=title K2Node_Event}
   > "SetText (Text)"(self=Get Text_Title() -> (Text_Title:Common Text Object Reference), InText="LOCTEXT(\"ATTRIBUTIONS\")")
   Event OnCreditsModalInitialized() {is=title K2Node_Event}
   > "SetText (Text)"(self=Get Text_Title(), InText="LOCTEXT(\"CREDITS\")")
  ]}
 GetAttributionsTextFormat {nodes [
   n1 = GetAttributionsTextFormat!() -> (then, Title:"Text (by ref)", Author:"Text (by ref)", License:"Text (by ref)") {is=title K2Node_FunctionEntry}
   > Return Node!(
    execute
    ReturnValue:Text=Format Text!(
     Format="LOCTABLE(\"/Game/UI/DiscoveryShared/StringTable_ActivityBrowser\", \"Attributions_Description\")"
     Title:"Text (by ref)"=@n1
     Author:"Text (by ref)"=@n1.Author
     License:"Text (by ref)"=@n1.License
    ) -> (Result)
   )
  ]}
 GetCreatorTextFormat {nodes [
   n1 = GetCreatorTextFormat!() -> (then, CreatorName:"Text (by ref)") {is=title K2Node_FunctionEntry}
   > Return Node(ReturnValue=Format Text!(Format="LOCTABLE(\"&StringTable_ActivityBrowser\", \"Details_CreatorFormat\")", Creator:"Text (by ref)"=@n1) -> (Result))
  ]}
 GetLinkCodeTextFormat {nodes [
   n1 = GetLinkCodeTextFormat!() -> (then, LinkCode:"Text (by ref)") {is=title K2Node_FunctionEntry}
   > Return Node(ReturnValue=Format Text!(
     Format="LOCTABLE(\"&StringTable_ActivityBrowser\", \"Details_LinkCodeFormat\")"
     XXXX-XXXX-XXXX-XXXX:"Text (by ref)"=@n1
    ) -> (Result))
  ]}
]
variables [IntroWhooshSound {type=SoundBase, default=/Game/Sounds/Fort_UI_Sounds/FrontEnd/MultiProduct/Common/Whooshes/Screens/MSS_UI_MP_Whoosh_Screen_Generic}]
widgets {
 info {
  inheritedWidgetCount=7
  namedSlotCount=2
  parentClass=/Script/FortniteUI.FortActivityBrowserAttributionsModal
  rootWidgetClass=/FortUILibrary/Systems/Dialog/WBP_UIKit_Dialog_Internal.WBP_UIKit_Dialog_Internal_C
  widgetCount=20
 }
 widgets [
  Dialog_Internal {
   widgetClassPath=/FortUILibrary/Systems/Dialog/WBP_UIKit_Dialog_Internal
   bIsVariable=true
   widget=Dialog_Internal
   properties {
    useClickCatch=true
    dialogSIze {
     height Override {enable=true, value=800}
     max Aspect Ratio {enable=false, value=0}
     max Desired Height {enable=false, value=0}
     max Desired Width {enable=false, value=0}
     min Aspect Ratio {enable=false, value=0}
     min Desired Height {enable=false, value=0}
     min Desired Width {enable=false, value=0}
     width Override {enable=true, value=1596}
    }
   }
  }
  Overlay_Modal {widgetClassPath=&Overlay, namedSlotHost=Dialog_Internal, widget=Overlay_Modal}
  Panel_Content {
   widgetClassPath=&HorizontalBox
   parent=Overlay_Modal
   widget=Panel_Content
   slot {class=OverlaySlot, properties {horizontalAlignment=HAlign_Fill, verticalAlignment=VAlign_Fill}}
  }
  VerticalBox_73 {
   widgetClassPath=&VerticalBox
   parent=Panel_Content
   widget=VerticalBox_73
   slot {class=HorizontalBoxSlot, properties {padding=<ltrb 80 64 48 48>, size {sizeRule=Fill, value=1}}}
  }
  VerticalBox_0 {
   widgetClassPath=&VerticalBox
   parent=VerticalBox_73
   widget=VerticalBox_0
   slot {class=VerticalBoxSlot, properties {verticalAlignment=VAlign_Center, padding=<ltrb 0 0 64 0>, size {sizeRule=Fill, value=1}}}
  }
  StackBox_0 {
   widgetClassPath=&StackBox
   parent=VerticalBox_0
   widget=StackBox_0
   properties {orientation=Orient_Vertical}
   slot {class=VerticalBoxSlot, properties {padding=<ltrb 0 0 0 16>}}
  }
  Text_ActivityOrigin {
   widgetClassPath=&CommonRichTextBlock
   bInherited=true
   bIsVariable=true
   parent=StackBox_0
   widget=Text_ActivityOrigin
   properties {
    mobileTextBlockScale=1.3
    text="Created by <CreatorStyle>{Creator}</> "
    textStyleSet=/Game/UI/DiscoveryShared/DT_UI_RichDetails_Tags
    defaultTextStyleOverride {
     overflowPolicy=Clip
     transformPolicy=None
     colorAndOpacity {colorUseRule=UseColor_Specified, specifiedColor=<rgba 0.0331 0.402 1 1>}
     font {
      bForceMonospaced=false
      bMaterialIsStencil=false
      fontMaterial=null
      fontObject=&BurbankSmall
      letterSpacing=0
      monospacedWidth=1
      size=15
      skewAmount=0
      typefaceFontName=Bold
      outlineSettings {
       bApplyOutlineToDropShadows=false
       bMiteredCorners=false
       bSeparateFillAlpha=false
       outlineMaterial=null
       outlineSize=0
       outlineColor=<rgba 0 0 0 1>
      }
     }
     highlightColor {colorUseRule=UseColor_Specified, specifiedColor=<rgba 0 0 0 1>}
     highlightShape {
      bIsDynamicallyLoaded=false
      drawAs=Image
      imageType=NoImage
      mirroring=NoMirror
      resourceName=None
      resourceObject=null
      tiling=NoTile
      imageSize=<32 32>
      margin=<ltrb 0 0 0 0>
      outlineSettings {
       bUseBrushTransparency=false
       roundingType=HalfHeightRadius
       width=0
       color {colorUseRule=UseColor_Specified, specifiedColor=<rgba 0 0 0 0>}
       cornerRadii=<0 0 0 0>
      }
      tintColor {colorUseRule=UseColor_Specified, specifiedColor=<rgba 1 1 1 1>}
      uVRegion {bIsValid=false, max=<0 0>, min=<0 0>}
     }
     selectedBackgroundColor {colorUseRule=UseColor_Foreground, specifiedColor=<rgba 1 0 1 1>}
     shadowColorAndOpacity=<rgba 0 0 0 1>
     shadowOffset=<0 0>
     strikeBrush {
      bIsDynamicallyLoaded=false
      drawAs=Image
      imageType=NoImage
      mirroring=NoMirror
      resourceName=None
      resourceObject=null
      tiling=NoTile
      imageSize=<32 32>
      margin=<ltrb 0 0 0 0>
      outlineSettings {
       bUseBrushTransparency=false
       roundingType=HalfHeightRadius
       width=0
       color {colorUseRule=UseColor_Specified, specifiedColor=<rgba 0 0 0 0>}
       cornerRadii=<0 0 0 0>
      }
      tintColor {colorUseRule=UseColor_Specified, specifiedColor=<rgba 1 1 1 1>}
      uVRegion {bIsValid=false, max=<0 0>, min=<0 0>}
     }
     underlineBrush {
      bIsDynamicallyLoaded=false
      drawAs=Image
      imageType=NoImage
      mirroring=NoMirror
      resourceName=None
      resourceObject=null
      tiling=NoTile
      imageSize=<32 32>
      margin=<ltrb 0 0 0 0>
      outlineSettings {
       bUseBrushTransparency=false
       roundingType=HalfHeightRadius
       width=0
       color {colorUseRule=UseColor_Specified, specifiedColor=<rgba 0 0 0 0>}
       cornerRadii=<0 0 0 0>
      }
      tintColor {colorUseRule=UseColor_Specified, specifiedColor=<rgba 1 1 1 1>}
      uVRegion {bIsValid=false, max=<0 0>, min=<0 0>}
     }
    }
    margin=<ltrb 0 -4 0 -4>
   }
   slot {class=StackBoxSlot, properties {verticalAlignment=VAlign_Top, padding=<ltrb 0 0 0 8>}}
  }
  Text_LinkCode {
   widgetClassPath=&CommonRichTextBlock
   bInherited=true
   bIsVariable=true
   parent=StackBox_0
   widget=Text_LinkCode
   properties {
    mobileTextBlockScale=1.3
    text="Island Code <CreatorStyle>{XXXX-XXXX-XXXX-XXXX}</> "
    textStyleSet=&DT_UI_RichDetails_Tags
    defaultTextStyleOverride {
     overflowPolicy=Clip
     transformPolicy=None
     colorAndOpacity {colorUseRule=UseColor_Specified, specifiedColor=<rgba 0.0331 0.402 1 1>}
     font {
      bForceMonospaced=false
      bMaterialIsStencil=false
      fontMaterial=null
      fontObject=&BurbankSmall
      letterSpacing=0
      monospacedWidth=1
      size=15
      skewAmount=0
      typefaceFontName=Bold
      outlineSettings {
       bApplyOutlineToDropShadows=false
       bMiteredCorners=false
       bSeparateFillAlpha=false
       outlineMaterial=null
       outlineSize=0
       outlineColor=<rgba 0 0 0 1>
      }
     }
     highlightColor {colorUseRule=UseColor_Specified, specifiedColor=<rgba 0 0 0 1>}
     highlightShape {
      bIsDynamicallyLoaded=false
      drawAs=Image
      imageType=NoImage
      mirroring=NoMirror
      resourceName=None
      resourceObject=null
      tiling=NoTile
      imageSize=<32 32>
      margin=<ltrb 0 0 0 0>
      outlineSettings {
       bUseBrushTransparency=false
       roundingType=HalfHeightRadius
       width=0
       color {colorUseRule=UseColor_Specified, specifiedColor=<rgba 0 0 0 0>}
       cornerRadii=<0 0 0 0>
      }
      tintColor {colorUseRule=UseColor_Specified, specifiedColor=<rgba 1 1 1 1>}
      uVRegion {bIsValid=false, max=<0 0>, min=<0 0>}
     }
     selectedBackgroundColor {colorUseRule=UseColor_Foreground, specifiedColor=<rgba 1 0 1 1>}
     shadowColorAndOpacity=<rgba 0 0 0 1>
     shadowOffset=<0 0>
     strikeBrush {
      bIsDynamicallyLoaded=false
      drawAs=Image
      imageType=NoImage
      mirroring=NoMirror
      resourceName=None
      resourceObject=null
      tiling=NoTile
      imageSize=<32 32>
      margin=<ltrb 0 0 0 0>
      outlineSettings {
       bUseBrushTransparency=false
       roundingType=HalfHeightRadius
       width=0
       color {colorUseRule=UseColor_Specified, specifiedColor=<rgba 0 0 0 0>}
       cornerRadii=<0 0 0 0>
      }
      tintColor {colorUseRule=UseColor_Specified, specifiedColor=<rgba 1 1 1 1>}
      uVRegion {bIsValid=false, max=<0 0>, min=<0 0>}
     }
     underlineBrush {
      bIsDynamicallyLoaded=false
      drawAs=Image
      imageType=NoImage
      mirroring=NoMirror
      resourceName=None
      resourceObject=null
      tiling=NoTile
      imageSize=<32 32>
      margin=<ltrb 0 0 0 0>
      outlineSettings {
       bUseBrushTransparency=false
       roundingType=HalfHeightRadius
       width=0
       color {colorUseRule=UseColor_Specified, specifiedColor=<rgba 0 0 0 0>}
       cornerRadii=<0 0 0 0>
      }
      tintColor {colorUseRule=UseColor_Specified, specifiedColor=<rgba 1 1 1 1>}
      uVRegion {bIsValid=false, max=<0 0>, min=<0 0>}
     }
    }
    margin=<ltrb 0 -4 0 -4>
   }
   slot {class=StackBoxSlot, properties {verticalAlignment=VAlign_Top}}
  }
  Text_ActivityName {
   widgetClassPath=&CommonTextBlock
   bInherited=true
   parent=VerticalBox_0
   widget=Text_ActivityName
   properties {
    applyLineHeightToBottomLine=false
    autoWrapText=true
    bIsScrollingEnabled=false
    text="{Game Mode}"
    wrappingPolicy=AllowPerCharacterWrapping
    font {
     bForceMonospaced=false
     bMaterialIsStencil=false
     fontMaterial=null
     fontObject=&HeadingNow
     letterSpacing=0
     monospacedWidth=1
     size=41.25
     skewAmount=0
     typefaceFontName=86 Bold Italic
     outlineSettings {
      bApplyOutlineToDropShadows=false
      bMiteredCorners=false
      bSeparateFillAlpha=false
      outlineMaterial=/FortUILibrary/MaterialInstances/Text/MI_UI_Block_Text_Outline
      outlineSize=4
      outlineColor=<rgba 0 0 0 1>
     }
    }
    shadowOffset=<0 0>
   }
   slot {class=VerticalBoxSlot}
  }
  Button_Back {
   widgetClassPath=&WBP_UIKit_Button_Regular
   bInherited=true
   bIsVariable=true
   parent=VerticalBox_73
   widget=Button_Back
   properties {bIsFocusable=false, text=Back, triggeringInputAction {dataTable=&InputActionDataTable, rowName=GenericBack}}
   slot {class=VerticalBoxSlot, properties {horizontalAlignment=HAlign_Left, verticalAlignment=VAlign_Top}}
  }
  Overlay_0 {
   widgetClassPath=&Overlay
   parent=Panel_Content
   widget=Overlay_0
   slot {class=HorizontalBoxSlot, properties {size {sizeRule=Fill, value=1}}}
  }
  Backplate_Attributions {
   widgetClassPath=&WBP_UIKit_Backplate
   bIsVariable=true
   parent=Overlay_0
   widget=Backplate_Attributions
   properties {brightness=Obscure}
   slot {class=OverlaySlot, properties {horizontalAlignment=HAlign_Fill, verticalAlignment=VAlign_Fill}}
  }
  ScrollBox_Details {
   widgetClassPath=&ScrollBox
   bInherited=true
   parent=Overlay_0
   widget=ScrollBox_Details
   slot {class=OverlaySlot, properties {horizontalAlignment=HAlign_Fill, verticalAlignment=VAlign_Fill, padding=<ltrb 48 64 80 48>}}
  }
  VerticalBox_17 {widgetClassPath=&VerticalBox, parent=ScrollBox_Details, widget=VerticalBox_17, slot {class=ScrollBoxSlot}}
  Text_Title {
   widgetClassPath=&CommonTextBlock
   bIsVariable=true
   parent=VerticalBox_17
   widget=Text_Title
   properties {
    applyLineHeightToBottomLine=false
    mobileFontSizeMultiplier=1.25
    scrollStyle=&TextScrollStyle-Base_C
    text=Attributions
    textTransformPolicy=ToUpper
    font {
     bForceMonospaced=false
     bMaterialIsStencil=false
     fontMaterial=null
     fontObject=&HeadingNow
     letterSpacing=0
     monospacedWidth=1
     size=22.5
     skewAmount=0
     typefaceFontName=75 Medium
     outlineSettings {
      bApplyOutlineToDropShadows=false
      bMiteredCorners=false
      bSeparateFillAlpha=false
      outlineMaterial=null
      outlineSize=0
      outlineColor=<rgba 0 0 0 1>
     }
    }
    shadowOffset=<0 0>
   }
   slot {class=VerticalBoxSlot, properties {verticalAlignment=VAlign_Top, padding=<ltrb 0 0 0 16>}}
  }
  RichText_Description {
   widgetClassPath=&CommonRichTextBlock
   bInherited=true
   bIsVariable=true
   parent=VerticalBox_17
   widget=RichText_Description
   properties {
    applyLineHeightToBottomLine=false
    autoWrapText=true
    mobileTextBlockScale=1.34
    text="\"Creative Commons 10th Birthday Celebration San Francisco\" by tvol is licensed under CC BY 2.0\r\n\r\nPhoto by tvol / CC BY\r\n\r\n\"Creative Commons 10th Birthday Celebration San Francisco\" by tvol, used under CC BY\r\n\r\n\"Creative Commons 10th Birthday San Francisco\" by tvol is licensed under CC BY 2.0 \r\n\r\n\"Creative Commons 10th Birthday Celebration Francisco\" by tvol is licensed under CC BY 2.0\r\n\r\nPhoto by tvol / CC BY\r\n\r\n\"Creative  Birthday Celebration San Francisco\" by tvol, used under CC BY"
    textStyleSet=&/AttributionStyles
    defaultTextStyleOverride {
     overflowPolicy=Clip
     transformPolicy=None
     colorAndOpacity {colorUseRule=UseColor_Specified, specifiedColor=<rgba 1 1 1 1>}
     font {
      bForceMonospaced=false
      bMaterialIsStencil=false
      fontMaterial=null
      fontObject=&NotoSans
      letterSpacing=0
      monospacedWidth=1
      size=24
      skewAmount=0
      typefaceFontName=NotoSans-Regular
      outlineSettings {
       bApplyOutlineToDropShadows=false
       bMiteredCorners=false
       bSeparateFillAlpha=false
       outlineMaterial=null
       outlineSize=0
       outlineColor=<rgba 0 0 0 1>
      }
     }
     highlightColor {colorUseRule=UseColor_Specified, specifiedColor=<rgba 0 0 0 1>}
     highlightShape {
      bIsDynamicallyLoaded=false
      drawAs=Image
      imageType=NoImage
      mirroring=NoMirror
      resourceName=None
      resourceObject=null
      tiling=NoTile
      imageSize=<32 32>
      margin=<ltrb 0 0 0 0>
      outlineSettings {
       bUseBrushTransparency=false
       roundingType=HalfHeightRadius
       width=0
       color {colorUseRule=UseColor_Specified, specifiedColor=<rgba 0 0 0 0>}
       cornerRadii=<0 0 0 0>
      }
      tintColor {colorUseRule=UseColor_Specified, specifiedColor=<rgba 1 1 1 1>}
      uVRegion {bIsValid=false, max=<0 0>, min=<0 0>}
     }
     selectedBackgroundColor {colorUseRule=UseColor_Foreground, specifiedColor=<rgba 1 0 1 1>}
     shadowColorAndOpacity=<rgba 0 0 0 1>
     shadowOffset=<0 0>
     strikeBrush {
      bIsDynamicallyLoaded=false
      drawAs=Image
      imageType=NoImage
      mirroring=NoMirror
      resourceName=None
      resourceObject=null
      tiling=NoTile
      imageSize=<32 32>
      margin=<ltrb 0 0 0 0>
      outlineSettings {
       bUseBrushTransparency=false
       roundingType=HalfHeightRadius
       width=0
       color {colorUseRule=UseColor_Specified, specifiedColor=<rgba 0 0 0 0>}
       cornerRadii=<0 0 0 0>
      }
      tintColor {colorUseRule=UseColor_Specified, specifiedColor=<rgba 1 1 1 1>}
      uVRegion {bIsValid=false, max=<0 0>, min=<0 0>}
     }
     underlineBrush {
      bIsDynamicallyLoaded=false
      drawAs=Image
      imageType=NoImage
      mirroring=NoMirror
      resourceName=None
      resourceObject=null
      tiling=NoTile
      imageSize=<32 32>
      margin=<ltrb 0 0 0 0>
      outlineSettings {
       bUseBrushTransparency=false
       roundingType=HalfHeightRadius
       width=0
       color {colorUseRule=UseColor_Specified, specifiedColor=<rgba 0 0 0 0>}
       cornerRadii=<0 0 0 0>
      }
      tintColor {colorUseRule=UseColor_Specified, specifiedColor=<rgba 1 1 1 1>}
      uVRegion {bIsValid=false, max=<0 0>, min=<0 0>}
     }
    }
   }
   slot {class=VerticalBoxSlot}
  }
  Overlay_57 {widgetClassPath=&Overlay, namedSlotHost=Dialog_Internal, widget=Overlay_57}
  SZBottomButtons {
   widgetClassPath=&SafeZone
   parent=Overlay_57
   widget=SZBottomButtons
   properties {padTop=false}
   slot {class=OverlaySlot, properties {horizontalAlignment=HAlign_Right, verticalAlignment=VAlign_Bottom}}
  }
  HBBottomButtons {
   widgetClassPath=&HorizontalBox
   parent=SZBottomButtons
   widget=HBBottomButtons
   slot {class=SafeZoneSlot, properties {padding=<ltrb 0 0 80 32>}}
  }
  Button_OpenLink {
   widgetClassPath=&WBP_UIKit_Button_Regular
   bInherited=true
   bIsVariable=true
   parent=HBBottomButtons
   widget=Button_OpenLink
   properties {
    bIsFocusable=false
    buttonSize=Small
    buttonType=Quiet
    clickMethod=MouseDown
    pressMethod=ButtonPress
    text=View on Fortnite.com
    visibility=Collapsed
    triggeringInputAction {dataTable=&InputActionDataTable, rowName=GenericRightStickClick}
   }
   slot {class=HorizontalBoxSlot, properties {horizontalAlignment=HAlign_Left, verticalAlignment=VAlign_Bottom}}
  }
 ]
}
